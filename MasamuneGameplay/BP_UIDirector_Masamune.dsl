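; UI director that shows the Masamune UI scene while a watched actor has the Masamune gameplay tag, skipping touch play.
; On BeginPlay it watches the owning player controller, then starts tag add and remove waits on each new view target.
; Tag changes drive SetMasamuneSceneActive, where a delta gate passes only state changes.
; It adds the BR Game UI or standard scene depending on that setting, and removes the stored scene on deactivation.
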(event EventBeginPlay
  (if (not (TouchPlatform|IsTouchInputPlatform))
    (bind _asfort_player_controller (Utilities|Casting|CastToFortPlayerController (DynamicUIDirector|GetOwningLocalPlayerController)))
    (EventDispatchers|BindEventtoOnViewTargetChanged _asfort_player_controller (delegate Custom|OnViewTargetChanged))
    (CallFunction|WaitforGameplayTag _asfort_player_controller)))

(event Custom|OnViewTargetChanged
  (params (PlayerController :type Fort%20Player%20Controller%20Object%20Reference) (OldViewTarget :type Actor%20Object%20Reference) (NewViewTarget :type Actor%20Object%20Reference))
  (CallFunction|WaitforGameplayTag NewViewTarget))

(event Custom|WaitForGameplayTag
  (params (PlayerActor :type Actor%20Object%20Reference))
  (bind _returnvalue (DynamicUIDirector|GetOwningLocalPlayerController))
  ;[1 1 1] Wait For Appropriate Tag
  (Utilities|IsValid PlayerActor
    (:Is%20Valid
      (Utilities|IsValid _returnvalue
        (:Is%20Valid
          (if (not (and (TouchPlatform|IsMobileGame) (TouchPlatform|IsUsingTouch _returnvalue)))
            (Utilities|Casting|CastToFortPlayerPawnAthena (Pawn|GetControlledPawn _returnvalue))
            (Ability|Tasks|WaitGameplayTagAddToActor PlayerActor (Variables|Default|GetMasamune%20Tag%20Event)
              (:then
                (Ability|Tasks|WaitGameplayTagRemoveFromActor PlayerActor (Variables|Default|GetMasamune%20Tag%20Event)
                  (:Removed
                    (CallFunction|SetMasamuneSceneActive false))))
              (:Added
                (CallFunction|SetMasamuneSceneActive true)))))))))

(fn SetMasamuneSceneActive
  (params (IsActive :type Boolean))
  (bind _updated_variable (Utilities|FlowControl|DeltaGateWildcard (Variables|Default|GetIsMasamuneSceneActive) IsActive)
    (:Success
      (if _updated_variable
        (bind _isenabled (Class|BPFLUISettingsHelpers|IsBRGameUIEnabled))
        (bind _output_get (Variables|Default|SetActiveUIScene (select _isenabled "/MasamuneGameplay/UI/Scene_MasamuneUI_BRGameUI.Scene_MasamuneUI_BRGameUI" "/MasamuneGameplay/UI/MasamuneUIScene.MasamuneUIScene")))
        (DynamicUIDirector|AddScene _output_get)
        (else
          (DynamicUIDirector|RemoveScene (Variables|Default|GetActiveUIScene)))))))

(fn ConstructionScript)
